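# Additive, two-sided unlit material for aurora-style rift bubble effects on the Rift Beacon, usable on Niagara mesh particles.
# Panning noise and radial sine waves displace vertices, with vertex color masking height and offsetting time.
# Noise blends time-of-day gradient colors and shapes opacity with UV masks, angle, optional depth/near-camera fades, UV1 arc gradients and MPC RiftOpacity.
# Mobile and forward paths substitute flat noise, skip fades and lower the top.
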Material /Game/Environments/Props/RiftBeacon/FX/M_Auroras_RiftBubbles
graph {
 properties {
  bUsedWithNiagaraMeshParticles=true
  blendMode=BLEND_Additive
  pixelDepthOffsetMode=PDOM_Legacy
  shadingModel=MSM_Unlit
  twoSided=true
  usedShadingModels=Unlit
 }
 nodes [
  n1 = TextureCoordinate()
  n2 = VectorParameter() {parameterName=NoiseTexScale2, defaultValue=<rgba 0.5 0 0 0>}
  n3 = AppendVector(A=@n2.R, B=@n2.R)
  n4 = Add(A=Time(), B=VertexColor().B)
  n5 = Multiply(A=ScalarParameter() {defaultValue=2, parameterName=HeightVariancePanSpeed}, B=@n4)
  n6 = TextureSample(UVs=Panner(Coordinate=Multiply(A=@n1, B=@n3), Time=@n5) {speedX=0.02, speedY=0.01}) {samplerType=SAMPLERTYPE_Grayscale, texture=&LowResBlurredNoise}
  n7 = LinearInterpolate(Alpha=@n6.G) {constA=0.2}
  n8 = VertexColor()
  n9 = Constant2Vector()
  n10 = ScalarParameter() {defaultValue=200000, parameterName=HeightVariance}
  n11 = ComponentMask(None=WorldPosition().XYZ) {g=true, r=true}
  n12 = ComponentMask(None=ObjectPositionWS()) {g=true, r=true}
  n13 = Add(
   A=Divide(A=Distance(A=@n11, B=@n12), B=ScalarParameter() {defaultValue=150000, parameterName=WaveLength})
   B=Multiply(A=@n4, B=ScalarParameter() {defaultValue=0.08, parameterName=WaveSpeed})
  )
  n14 = Constant3Vector() {constant=<rgba 1 1 0 0>}
  n15 = Multiply(
   A=Multiply(A=Sine(None=@n13), B=Transform(None=@n14) {transformSourceType=TRANSFORMSOURCE_Local})
   B=ScalarParameter() {defaultValue=25, parameterName=WaveStrength}
  )
  n16 = DotProduct(
   A=MaterialFunctionCall() {materialFunction=&CameraDirectionVector}.Result
   B=VertexNormalWS()
  )
  n17 = Clamp(None=Power(Base=Abs(None=@n16)) {constExponent=4})
  n18 = TextureCoordinate()
  n19 = Multiply(
   A=Multiply(A=ScalarParameter() {defaultValue=-3000, parameterName=mobileLowerTopAmount}, B=Constant3Vector() {constant=<rgba 0 0 1 0>})
   B=OneMinus(None=ComponentMask(None=@n18) {g=true})
  )
  Add(
   A=Multiply(A=Multiply(A=@n7, B=@n8.R), B=AppendVector(A=@n9, B=@n10))
   B=ShadingPathSwitch(Default=Multiply(A=@n15, B=@n17), Forward=@n19, Mobile=@n19)
  )
  Constant2Vector()
  n20 = VectorParameter() {parameterName=TopColor, defaultValue=<rgba 0.0322 0.1139 0.3 0>}
  n21 = VectorParameter() {parameterName=TopColor_Day, defaultValue=<rgba 0.0322 0.1139 0.3 0>}
  n22 = VectorParameter() {parameterName=BottomColor, defaultValue=<rgba 0.6165 2 0 0>}
  n23 = VectorParameter() {parameterName=BottomColor_Day, defaultValue=<rgba 0.6165 2 0 0>}
  n24 = TextureObjectParameter() {parameterName=WorldAlignedTexture, texture=&T_Smooth_tile}
  n25 = ScalarParameter() {defaultValue=80000, parameterName=WorldAlignedTexScale}
  n26 = Time()
  n27 = VectorParameter() {parameterName=WorldAlignedPanDirection, defaultValue=<rgba -0.7 -0.4 0 1>}
  n28 = Add(
   A=Multiply(A=Multiply(A=@n26, B=@n27.RGB), B=ScalarParameter() {defaultValue=11000, parameterName=WorldAlignedPanSpeed})
   B=WorldPosition().XYZ
  )
  n29 = ScalarParameter() {defaultValue=0.6, parameterName=MobileFlat}
  n30 = ShadingPathSwitch(
   Default=ComponentMask(None=MaterialFunctionCall(TextureObject=@n24, TextureSize=@n25, WorldPosition=@n28) {materialFunction=&WorldAlignedTexture}.Z Texture) {g=true}
   Forward=@n29
   Mobile=@n29
  )
  n31 = VertexColor()
  n32 = VectorParameter() {parameterName=NoiseTexScale, defaultValue=<rgba 4 0.8 0 0>}
  n33 = Multiply(A=TextureCoordinate(), B=AppendVector(A=@n32.R, B=@n32.R))
  n34 = ShadingPathSwitch(
   Default=TextureSample(UVs=Panner(Coordinate=@n33, Time=@n5) {speedX=0.04, speedY=0.12}) {samplerType=SAMPLERTYPE_Grayscale, texture=&LowResBlurredNoise}.G
   Forward=@n29
   Mobile=@n29
  )
  n35 = LinearInterpolate(A=ScalarParameter() {defaultValue=0.3, parameterName=Min}, Alpha=Multiply(A=LinearInterpolate(A=@n30, Alpha=@n31.G), B=@n34)) {constA=0.4}
  n36 = ComponentMask(None=TextureCoordinate()) {g=true}
  n37 = LinearInterpolate(Alpha=LinearInterpolate(A=@n30, Alpha=@n34)) {constA=2, constB=0.9}
  n38 = Power(Base=@n36, Exp=Multiply(A=ScalarParameter() {defaultValue=5, parameterName=ColorLerpPower}, B=@n37)) {constExponent=4}
  LinearInterpolate(
   A=LinearInterpolate(
    A=MaterialFunctionCall(Color_Morning=@n20.RGB, Color_Day=@n21.RGB, Color_Evening=@n20.RGB, Color_Night=@n20.RGB) {materialFunction=&Color_Time_of_Day}.Result
    B=MaterialFunctionCall(Color_Morning=@n22.RGB, Color_Day=@n23.RGB, Color_Evening=@n22.RGB, Color_Night=@n22.RGB) {materialFunction=&Color_Time_of_Day}.Result
    Alpha=Multiply(A=@n35, B=@n38)
   )
   B=VectorParameter() {parameterName=EdgeColor, defaultValue=<rgba 0.0709 0.7 0.2183 0>}.RGB
   Alpha=VertexColor().G
  )
  Multiply()
  n39 = ScalarParameter() {defaultValue=0.25, parameterName=Opacity}
  n40 = ScalarParameter() {defaultValue=0.97, parameterName=BottomHeightVariance}
  n41 = LinearInterpolate(A=@n34, Alpha=@n31.G)
  n42 = SphereMask(
   A=ComponentMask(None=TextureCoordinate() {vTiling=0.5}) {g=true}
   B=Constant()
   Radius=Multiply(B=LinearInterpolate(A=@n40, Alpha=@n41) {constA=0.95}) {constA=0.5}
   Hardness=ScalarParameter() {defaultValue=0.91, parameterName=MaskHardness}
  )
  n43 = Power(Base=Power(Base=Multiply(A=@n36, B=@n42)) {constExponent=1.5}, Exp=@n37)
  n44 = TextureCoordinate() {vTiling=0.5}
  n45 = Constant() {r=0.5}
  n46 = DotProduct(A=CameraVectorWS(), B=VertexNormalWS())
  n47 = Clamp(None=Power(Base=Abs(None=@n46), Exp=ScalarParameter() {defaultValue=1.4, parameterName=AngleFade}))
  n48 = Constant() {r=1}
  n49 = Multiply(
   A=SphereMask(A=ComponentMask(None=@n44) {r=true}, B=@n45, Radius=@n45) {hardnessPercent=70}
   B=ShadingPathSwitch(Default=ShadingPathSwitch(Default=@n47, Mobile=@n48) {desc=Migrated from FeatureLevelSwitch}, Forward=@n48)
  )
  n50 = Multiply(
   A=MaterialFunctionCall(
    Color_Morning=ScalarParameter() {defaultValue=0.25, parameterName=Opacity_Morning}
    Color_Day=@n39
    Color_Evening=@n39
    Color_Night=ScalarParameter() {defaultValue=0.25, parameterName=Opacity_Night}
   ) {materialFunction=&Color_Time_of_Day}.Result
   B=Multiply(A=@n35, B=Multiply(A=@n43, B=@n49))
  )
  n51 = DepthFade(Opacity=@n50, FadeDistance=ScalarParameter() {defaultValue=500, parameterName=Depth Fade Distance})
  n52 = MaterialFunctionCall(Distance=ScalarParameter() {defaultValue=250, parameterName=Near Camera Distance}) {materialFunction=&Near_Camera_Fade}
  n53 = ShadingPathSwitch(
   Default=StaticSwitchParameter(True=Multiply(A=@n51, B=@n52.Result), False=@n50) {parameterName=USe Near and Depth Fade}
   Forward=@n50
   Mobile=@n50
  )
  n54 = Multiply(A=Time() {period=1}, B=ScalarParameter() {defaultValue=0.05, parameterName=UV1PannerScale}) {constB=9}
  n55 = Panner(
   Coordinate=MaterialFunctionCall(XY=TextureCoordinate() {coordinateIndex=1}) {materialFunction=&Swizzle}.YX
   Time=@n54
  ) {speedY=-1}
  n56 = ScalarParameter() {defaultValue=5, parameterName=ArcIntensity}
  n57 = Multiply(
   A=Power(
    Base=TextureSampleParameter2D(UVs=@n55) {parameterName=ArcGradientTexture, texture=/Game/Global/Textures/Default/Masks_and_Gradients/gradient_stasis}.G
    Exp=ScalarParameter() {defaultValue=1, parameterName=ArcExpControl}
   )
   B=@n56
  ) {constB=5}
  n58 = Panner(
   Coordinate=MaterialFunctionCall(XY=TextureCoordinate() {coordinateIndex=1}) {materialFunction=&Swizzle}.YX
   Time=@n54
  ) {speedY=-1}
  n59 = Multiply(
   A=Power(
    Base=TextureSampleParameter2D(UVs=@n58) {parameterName=ArcGradientTexture, texture=&gradient_stasis}.G
    Exp=ScalarParameter() {defaultValue=1, parameterName=ArcExpControl}
   )
   B=@n56
  ) {constB=5}
  n60 = LinearInterpolate(B=Saturate(None=Add(A=@n57, B=@n59)), Alpha=ScalarParameter() {defaultValue=1, parameterName=UV1Alpha}) {constA=1}
  Multiply(
   A=CollectionParameter() {collection=/Game/Athena/Environments/Nexus/Materials/FXshader/MPC_Nexus, parameterName=RiftOpacity}
   B=Multiply(A=Multiply(A=@n53, B=@n60), B=ScalarParameter() {defaultValue=1, parameterName=Alpha})
  )
  ScalarParameter() {defaultValue=1, parameterName=Noise_TextureScale_B}
  ShadingPathSwitch(Default=@n6.G)
 ]
}
