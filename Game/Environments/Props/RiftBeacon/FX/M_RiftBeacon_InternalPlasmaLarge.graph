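# Translucent unlit material for the large internal plasma effect of the Rift Beacon prop, usable on Niagara mesh particles.
# Panning noise drives parallax bump offset of turbulence textures, blended by a V-based mask with a position-projected sample to cover UV extremes.
# Fresnel-indexed curve atlas ramps and noise drive a color ramp, two-tint gradient and opacity faded along V.
# A SmallPlasma switch simplifies color and adds panning normal displacement to the camera offset.
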Material /Game/Environments/Props/RiftBeacon/FX/M_RiftBeacon_InternalPlasmaLarge
graph {
 properties {
  bUsedWithNiagaraMeshParticles=true
  blendMode=BLEND_Translucent
  pixelDepthOffsetMode=PDOM_Legacy
  shadingModel=MSM_Unlit
  usedShadingModels=Unlit
 }
 nodes [
  n1 = ComponentMask(None=TextureCoordinate()) {g=true}
  n2 = ScalarParameter() {defaultValue=-1.5, parameterName=RA_wide}
  n3 = Time()
  n4 = ScalarParameter() {defaultValue=2, parameterName=Panning Speed}
  n5 = VertexNormalWS()
  n6 = CameraVectorWS()
  n7 = CurveAtlasRowParameter(CurveTime=Abs(None=DotProduct(A=@n5, B=@n6))) {
   atlas=&AthenaCurveAtlas01
   curve=/Game/Athena/Environments/Nexus/Materials/FXshader/Curve_Ramps
   defaultValue=79
  }
  n8 = Add(
   A=Multiply(A=@n7.G, B=ScalarParameter() {defaultValue=0.025, parameterName=FresnelAffectTexture})
   B=Add(A=Multiply(A=@n1, B=@n2), B=Multiply(A=@n3, B=@n4))
  )
  n9 = ScalarParameter() {defaultValue=1, parameterName=TextureUScale}
  n10 = ScalarParameter() {defaultValue=1, parameterName=TextureVScale}
  n11 = Multiply(
   A=TextureCoordinate() {uTiling=3}
   B=AppendVector(A=ScalarParameter() {defaultValue=1, parameterName=TextureUScale}, B=ScalarParameter() {defaultValue=1, parameterName=TextureVScale})
  )
  n12 = TextureSample(UVs=Panner(Coordinate=@n11) {speedX=0.2, speedY=0.2}) {samplerType=SAMPLERTYPE_LinearColor, texture=&T_NoisePack_LinearColor}
  n13 = TextureSample(UVs=Panner(Coordinate=@n11) {speedX=-0.2, speedY=-0.2}) {samplerType=SAMPLERTYPE_LinearColor, texture=&T_NoisePack_LinearColor}
  n14 = ComponentMask(None=TextureCoordinate() {uTiling=0.5, vTiling=0.5}) {g=true}
  n15 = MaterialFunctionCall(In=Abs(None=Sine(None=@n14)), Contrast=Constant() {r=0.1}) {materialFunction=&CheapContrast}
  n16 = Add(
   A=Multiply(A=Subtract(A=@n12.R, B=@n13.G), B=@n15.Result)
   B=Multiply(A=Subtract(A=@n7.R), B=MaterialFunctionCall() {materialFunction=&DebugTimeSine}.Result)
  )
  n17 = Multiply(A=@n15.Result, B=ScalarParameter() {defaultValue=0.25, parameterName=Height})
  n18 = BumpOffset(Coordinate=Multiply(A=TextureCoordinate() {uTiling=4, vTiling=2.5}, B=AppendVector(A=@n9, B=@n10)), Height=@n16, HeightRatioInput=@n17)
  n19 = Multiply(A=Time(), B=ScalarParameter() {defaultValue=1, parameterName=TexturePanningSpeed})
  n20 = Panner(Coordinate=@n18, Time=@n19) {speedX=0.1, speedY=-0.3}
  n21 = Panner(Coordinate=Multiply(A=@n18) {constB=1.618}, Time=@n19) {speedX=0.1, speedY=-0.3}
  n22 = WorldPosition()
  n23 = Divide(
   A=ComponentMask(None=TransformPosition(None=@n22.XYZ) {transformSourceType=TRANSFORMPOSSOURCE_World}) {g=true, r=true}
   B=ScalarParameter() {defaultValue=1200, parameterName=Size}
  )
  n24 = LinearInterpolate(
   A=Multiply(
    A=TextureSample(UVs=@n20) {texture=/Game/Global/Textures/Default/Caustics/T_Turbulence_Seamless}.R
    B=TextureSample(UVs=@n21) {samplerType=SAMPLERTYPE_LinearColor, texture=&T_NoisePack_LinearColor}.R
   )
   B=TextureSample(UVs=BumpOffset(Coordinate=@n23, Height=@n16, HeightRatioInput=@n17)) {texture=&T_Turbulence_Seamless}.R
   Alpha=OneMinus(None=@n15.Result)
  )
  n25 = Multiply(A=@n24, B=ScalarParameter() {defaultValue=0.5, parameterName=NoiseScale})
  n26 = CurveAtlasRowParameter(CurveTime=Frac(None=Add(A=@n8, B=@n25))) {
   atlas=&AthenaCurveAtlas01
   curve=/Game/Athena/Environments/Nexus/Materials/FXshader/Curve_Color
   defaultValue=78
  }
  n27 = Multiply(A=Multiply(A=Saturate(None=@n24), B=@n7.B)) {constB=5}
  Add(A=@n26.A, B=@n27)
  n28 = Multiply(A=@n27, B=VectorParameter() {parameterName=TextureTint, defaultValue=<rgba 0.0984 0.3942 1 1>}.RGB)
  n29 = VectorParameter() {parameterName=Tint, defaultValue=<rgba 2 0.3438 0.3438 0.5>}
  n30 = VectorParameter() {parameterName=Tint_B, defaultValue=<rgba 2.5781 9.1695 15 0.5>}
  n31 = TextureCoordinate()
  n32 = Power(Base=Sine(None=ComponentMask(None=@n31) {g=true}) {period=2}, Exp=ScalarParameter() {defaultValue=4, parameterName=ColorExpGradient})
  n33 = Power(
   Base=Multiply(A=@n26, B=LinearInterpolate(A=@n29.RGB, B=@n30.RGB, Alpha=@n32))
   Exp=ScalarParameter() {defaultValue=1.5, parameterName=EdgeExp}
  )
  Multiply(
   A=VectorParameter() {parameterName=SphereWave_Tint, defaultValue=<rgba 1 1 1 0>}.RGB
   B=StaticSwitchParameter(True=@n28, False=Add(A=@n33, B=@n28)) {parameterName=SmallPlasma}
  )
  n34 = Sine(None=ComponentMask(None=TextureCoordinate()) {g=true}) {period=2}
  n35 = ScalarParameter() {defaultValue=1, parameterName=OpacityExpGradient}
  Multiply(A=Multiply(A=Multiply(A=@n7.A, B=@n27), B=Power(Base=@n34, Exp=@n35)), B=ScalarParameter() {defaultValue=1, parameterName=Opacity})
  n36 = MaterialFunctionCall(Offset Amount=ScalarParameter() {defaultValue=10, parameterName=Camera_Offset}) {materialFunction=/Engine/Functions/Engine_MaterialFunctions02/WorldPositionOffset/CameraOffset}
  n37 = VertexNormalWS()
  n38 = LinearInterpolate(
   B=ScalarParameter() {defaultValue=-5, parameterName=WPO_Amount}
   Alpha=TextureSample(UVs=Panner() {speedY=-1}) {texture=/Game/Global/Textures/Default/Masks_and_Gradients/gradient_stasis}.R
  )
  StaticSwitchParameter(True=Add(A=@n36.Camera Offset, B=Multiply(A=@n37, B=@n38)), False=@n36.Clamped Camera Offset) {parameterName=SmallPlasma}
 ]
}
