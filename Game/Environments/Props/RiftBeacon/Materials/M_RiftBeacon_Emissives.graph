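# Opaque lit material for the Rift Beacon prop's animated emissive parts, usable on spline meshes, using vertex color to separate screens and energy surfaces from the textured body.
# Blue vertices show monitors: panning, rotating and flipbook masks, or a noisy virus screen via the RobotDamage switch.
# Red vertices get black, glossy energy with radial bump-offset noise, curve-atlas color ramps and flickering Fresnel.
# MF_Geode_Hoses_Lights_RiftBeacon post-processes the resulting attributes using the emissive mask.
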Material /Game/Environments/Props/RiftBeacon/Materials/M_RiftBeacon_Emissives
graph {
 properties {bUseMaterialAttributes=true, bUsedWithSplineMeshes=true, pixelDepthOffsetMode=PDOM_Legacy, usedShadingModels=Default Lit}
 nodes [
  n1 = TextureSample() {texture=/Game/Environments/Props/RiftBeacon/Textures/T_RiftBeaconA_Bits_D}
  n2 = VertexColor()
  n3 = TextureSample() {samplerType=SAMPLERTYPE_Masks, texture=/Game/Environments/Props/RiftBeacon/Textures/T_RiftBeaconA_Bits_S}
  n4 = VertexColor()
  n5 = VertexColor()
  n6 = TextureSample() {texture=/Game/Environments/Props/RiftBeacon/Textures/T_RiftBeaconA_Bits_E}
  n7 = CameraVectorWS()
  n8 = VertexNormalWS()
  n9 = VectorParameter() {parameterName=Color 1, defaultValue=<rgba 4 0 0 1>}
  n10 = TextureCoordinate() {coordinateIndex=1}
  n11 = ComponentMask(None=Constant3Vector() {constant=<rgba 1.1098 1.1543 0 0>}) {g=true, r=true}
  n12 = ComponentMask(None=Constant3Vector() {constant=<rgba -0.0667 -0.1143 0 1>}) {g=true, r=true}
  n13 = Multiply(A=@n10) {constB=0}
  n14 = Multiply(
   A=TextureSample(UVs=Panner(Coordinate=@n13) {speedX=0.1, speedY=0.1}) {texture=/Game/Global/Textures/Default/Noise/Noise}.R
   B=TextureSample(UVs=Panner(Coordinate=@n13) {speedX=-0.1, speedY=-0.1}) {texture=&Noise}.R
  )
  n15 = Multiply(A=ComponentMask(None=Constant3Vector() {constant=<rgba 0 0.02 0 0>}) {g=true, r=true}, B=@n14)
  n16 = TextureSample(UVs=Add(A=Multiply(A=@n10, B=@n11), B=Add(A=@n12, B=@n15))) {texture=/Game/Environments/Props/RiftBeacon/Textures/brute_virus_screen4}
  n17 = Multiply(A=@n16.R, B=@n14)
  n18 = Multiply(A=@n10, B=Constant())
  n19 = Panner(Coordinate=@n18) {speedX=0.3, speedY=0.3}
  n20 = Panner(Coordinate=@n18) {speedX=-0.3, speedY=-0.3}
  n21 = Multiply(
   A=@n16.B
   B=Multiply(
    A=TextureSample(UVs=@n19) {texture=&Noise}.R
    B=TextureSample(UVs=@n20) {texture=&Noise}.R
   )
  )
  n22 = VectorParameter() {parameterName=Color 2, defaultValue=<rgba 0.3125 1.5702 4 0>}
  n23 = Multiply(A=Add(A=Multiply(A=@n9.RGB, B=@n17), B=Multiply(A=@n21, B=@n22.RGB)), B=Multiply(A=@n16.RGB, B=Constant() {r=2}))
  n24 = Multiply(A=@n16.RGB, B=Constant() {r=0.1}) {constB=0.25}
  n25 = Add(
   A=TextureCoordinate() {coordinateIndex=1}
   B=AppendVector(A=ScalarParameter() {parameterName=Uoffset}, B=ScalarParameter() {parameterName=Voffset})
  )
  n26 = TextureSample(UVs=Panner(Coordinate=@n25) {speedY=-0.25}) {texture=/Game/Environments/Props/RiftBeacon/Textures/T_RiftBeacon_MonitorMask}
  n27 = TextureSample(UVs=@n25) {texture=/Game/Environments/Props/RiftBeacon/Textures/T_RiftBeacon_MonitorMaskB}
  n28 = Panner(Coordinate=@n25) {speedY=0.05}
  n29 = Add(
   A=Add(A=TextureSample(UVs=@n25) {texture=&T_RiftBeacon_MonitorMask}.R, B=Multiply(A=@n26.G, B=@n27.R))
   B=Multiply(
    A=TextureSample(UVs=@n28) {texture=&T_RiftBeacon_MonitorMask}.G
    B=TextureSample(UVs=@n25) {texture=&T_RiftBeacon_MonitorMaskB}.G
   )
  )
  n30 = Multiply(
   A=TextureSample(UVs=Panner(Coordinate=@n25) {speedY=0.15}) {texture=&T_RiftBeacon_MonitorMask}.G
   B=TextureSample(UVs=@n25) {texture=&T_RiftBeacon_MonitorMaskB}.A
  )
  n31 = TextureSample(UVs=Panner(Coordinate=@n25) {speedY=-0.15}) {texture=&T_RiftBeacon_MonitorMask}
  n32 = TextureSample(UVs=@n25) {texture=&T_RiftBeacon_MonitorMaskB}
  n33 = Rotator(Coordinate=TextureCoordinate() {coordinateIndex=1}) {speed=1.3}
  n34 = Rotator(Coordinate=TextureCoordinate() {coordinateIndex=1}) {speed=-2}
  n35 = Add(
   A=Add(A=Add(A=@n29, B=@n30), B=Multiply(A=@n31.G, B=@n32.B))
   B=Add(
    A=TextureSample(UVs=@n33) {texture=&T_RiftBeacon_MonitorMask}.B
    B=TextureSample(UVs=@n34) {texture=&T_RiftBeacon_MonitorMask}.A
   )
  )
  n36 = TextureSample(UVs=TextureCoordinate() {coordinateIndex=1}) {texture=/Game/Environments/Props/RiftBeacon/Textures/T_RiftBeacon_MonitorMaskC}
  n37 = Constant() {r=2}
  n38 = ScalarParameter() {defaultValue=0.05, parameterName=Param_12}
  n39 = ScalarParameter() {defaultValue=0.1, parameterName=Param_22}
  n40 = MaterialFunctionCall!(
   U_Images=@n37
   V_Images=@n37
   SubUV Texture RGB=TextureObject() {texture=&T_RiftBeacon_MonitorMaskC}
   Time Speed=ScalarParameter() {defaultValue=1, parameterName=Speed}
   Manual Time Control
   Time=Time() {period=1}
   UVs=Add(A=TextureCoordinate() {coordinateIndex=1}, B=AppendVector(A=@n38, B=@n39))
  ) {materialFunction=/Game/Effects/Fort_Effects/Materials/_Functions/MF_SubUV}
  n41 = Constant() {r=2}
  n42 = ScalarParameter() {defaultValue=-0.05, parameterName=Param_1}
  n43 = ScalarParameter() {defaultValue=-0.1, parameterName=Param_2}
  n44 = MaterialFunctionCall(
   U_Images=@n41
   V_Images=@n41
   SubUV Texture RGB=TextureObject() {texture=&T_RiftBeacon_MonitorMaskC}
   Time Speed=ScalarParameter() {defaultValue=1, parameterName=Speed}
   Time=Time() {period=1}
   UVs=Add(A=TextureCoordinate() {coordinateIndex=1}, B=AppendVector(A=@n42, B=@n43))
  ) {materialFunction=&MF_SubUV}
  n45 = Add(A=Multiply(A=@n36.B, B=ComponentMask(None=@n40.RGB) {r=true}), B=Multiply(A=@n36.A, B=ComponentMask(None=@n44.RGB) {g=true}))
  n46 = StaticSwitchParameter(
   True=Multiply(A=DotProduct(A=@n7, B=@n8), B=Add(A=@n23, B=@n24))
   False=Multiply(A=VectorParameter() {parameterName=TextColor, defaultValue=<rgba 1 0.0167 0 1>}.RGB, B=Add(A=@n35, B=@n45))
  ) {parameterName=RobotDamage}
  n47 = Multiply(A=LinearInterpolate(B=Multiply(A=@n6.RGB, B=@n46), Alpha=VertexColor().B), B=ScalarParameter() {parameterName=MonitorOn})
  n48 = ScalarParameter() {defaultValue=2, parameterName=Fresnel1}
  n49 = Multiply(A=Time(), B=ScalarParameter() {defaultValue=0.8, parameterName=FlickerSpeed})
  n50 = Sine(None=@n49) {period=0.55}
  n51 = Sine(None=@n49) {period=0.11}
  n52 = Add(A=Add(A=Multiply(B=@n50), B=Multiply(A=@n51) {constB=0.2}), B=Multiply(A=Sine(None=@n49) {period=0.175}) {constB=0.3})
  n53 = ConstantBiasScale(None=Multiply(A=Sine(None=@n49) {period=3.25}, B=Sine(None=@n49) {period=2.325}))
  n54 = Multiply(A=Multiply(A=Multiply(A=@n52, B=@n53)) {constB=0.3}) {constB=0.03}
  n55 = Multiply(
   A=VectorParameter() {parameterName=Outer, defaultValue=<rgba 0.2078 0.191 0.4844 1>}.RGB
   B=Fresnel(ExponentIn=Add(A=@n48, B=@n54)) {baseReflectFraction=0}
  )
  n56 = Multiply(A=Time(), B=ScalarParameter() {defaultValue=-1, parameterName=TexturePanningSpeed})
  n57 = AppendVector(A=ScalarParameter() {defaultValue=2, parameterName=U_Tile}, B=ScalarParameter() {defaultValue=2, parameterName=V_Tile})
  n58 = Multiply(A=TextureCoordinate() {coordinateIndex=1}, B=@n57)
  n59 = Divide(A=@n57)
  n60 = TextureCoordinate() {coordinateIndex=1}
  n61 = AppendVector(A=ScalarParameter() {defaultValue=1, parameterName=U_Tile_2}, B=ScalarParameter() {defaultValue=1, parameterName=V_Tile_2})
  n62 = MaterialFunctionCall(Vector or UVs=Subtract(A=Multiply(A=@n60, B=@n61), B=Divide(A=@n61))) {materialFunction=&VectorToRadialValue}
  n63 = TextureCoordinate()
  n64 = DDX(Value=@n63)
  n65 = DDY(Value=@n63)
  n66 = TextureSample!(UVs=Panner(Coordinate=@n62.Radial Coordinates) {speedX=0.2, speedY=0.2}, Tex, "DDX(UVs)"=@n64, "DDY(UVs)"=@n65, Apply View MipBias) {mipValueMode=TMVM_Derivative, samplerType=SAMPLERTYPE_LinearColor, texture=&T_NoisePack_LinearColor}
  n67 = TextureSample(UVs=Panner(Coordinate=@n62.Radial Coordinates) {speedX=-0.2, speedY=-0.2}, "DDX(UVs)"=@n64, "DDY(UVs)"=@n65) {mipValueMode=TMVM_Derivative, samplerType=SAMPLERTYPE_LinearColor, texture=&T_NoisePack_LinearColor}
  n68 = ComponentMask(None=TextureCoordinate() {coordinateIndex=1, uTiling=0.5, vTiling=0.5}) {g=true}
  n69 = MaterialFunctionCall(In=Abs(None=Sine(None=@n68)), Contrast=Constant() {r=0.1}) {materialFunction=&CheapContrast}
  n70 = VertexNormalWS()
  n71 = CameraVectorWS()
  n72 = CurveAtlasRowParameter(CurveTime=Abs(None=DotProduct(A=@n70, B=@n71))) {
   atlas=&AthenaCurveAtlas01
   curve=/Game/Athena/Environments/Nexus/Materials/FXshader/Curve_Ramps
   defaultValue=79
  }
  n73 = Add(
   A=Multiply(A=@n69.Result, B=Subtract(A=@n66.R, B=@n67.G))
   B=Multiply(A=Subtract(A=@n72.R), B=MaterialFunctionCall() {materialFunction=&DebugTimeSine}.Result)
  )
  n74 = Multiply(A=@n69.Result, B=ScalarParameter() {defaultValue=-0.25, parameterName=Height})
  n75 = BumpOffset(
   Coordinate=MaterialFunctionCall(Vector or UVs=Subtract(A=@n58, B=@n59)) {materialFunction=&VectorToRadialValue}.Radial Coordinates
   Height=@n73
   HeightRatioInput=@n74
  )
  n76 = Panner(Coordinate=@n75, Time=@n56) {speedX=0.1, speedY=-0.3}
  n77 = TextureCoordinate()
  n78 = DDX(Value=@n77)
  n79 = DDY(Value=@n77)
  n80 = TextureSample(UVs=Panner(Coordinate=Multiply(A=@n75), Time=@n56) {speedX=0.1, speedY=-0.3}, "DDX(UVs)"=@n78, "DDY(UVs)"=@n79) {mipValueMode=TMVM_Derivative, samplerType=SAMPLERTYPE_LinearColor, texture=&T_NoisePack_LinearColor}
  n81 = WorldPosition()
  n82 = Divide(
   A=ComponentMask(None=TransformPosition(None=@n81.XYZ) {transformSourceType=TRANSFORMPOSSOURCE_World}) {g=true, r=true}
   B=ScalarParameter() {defaultValue=1200, parameterName=Size}
  )
  n83 = LinearInterpolate(
   A=Multiply(
    A=TextureSample(UVs=@n76, "DDX(UVs)"=@n78, "DDY(UVs)"=@n79) {mipValueMode=TMVM_Derivative, texture=/Game/Global/Textures/Default/Caustics/T_Turbulence_Seamless}.R
    B=@n80.R
   )
   B=TextureSample!(UVs=BumpOffset(Coordinate=@n82, Height=@n73, HeightRatioInput=@n74), Tex, Apply View MipBias) {texture=&T_Turbulence_Seamless}.R
   Alpha=OneMinus(None=@n69.Result)
  )
  n84 = Saturate(None=@n83)
  n85 = Multiply(A=@n83, B=ScalarParameter() {defaultValue=0.5, parameterName=NoiseScale})
  n86 = AppendVector(A=ScalarParameter() {defaultValue=1, parameterName=U_Tile_3}, B=ScalarParameter() {defaultValue=1, parameterName=V_Tile_3})
  n87 = Multiply(A=TextureCoordinate() {coordinateIndex=1}, B=@n86)
  n88 = Divide(A=@n86)
  n89 = ComponentMask(None=MaterialFunctionCall(Vector or UVs=Subtract(A=@n87, B=@n88)) {materialFunction=&VectorToRadialValue}.Radial Coordinates) {g=true}
  n90 = ScalarParameter() {defaultValue=-2, parameterName=RA_wide}
  n91 = Time()
  n92 = ScalarParameter() {defaultValue=2, parameterName=Panning Speed}
  n93 = Add(
   A=Add(A=Multiply(A=@n89, B=@n90), B=Multiply(A=@n91, B=@n92))
   B=Multiply(A=@n72.G, B=ScalarParameter() {defaultValue=0.025, parameterName=FresnelAffectTexture})
  )
  n94 = CurveAtlasRowParameter(CurveTime=Frac(None=Add(A=@n85, B=@n93))) {
   atlas=&AthenaCurveAtlas01
   curve=/Game/Athena/Environments/Nexus/Materials/FXshader/Curve_Color
   defaultValue=78
  }
  n95 = Add(
   A=Multiply(A=Multiply(A=@n84) {constB=5}, B=VectorParameter() {parameterName=TextureTint, defaultValue=<rgba 0.0984 0.3942 1 1>}.RGB)
   B=Power(Base=@n94, Exp=ScalarParameter() {defaultValue=1, parameterName=EdgeExp})
  )
  n96 = Multiply(
   A=Multiply(A=Add(A=@n55, B=@n95), B=ScalarParameter() {defaultValue=0.3, parameterName=RedVert_EmissScale})
   B=ScalarParameter() {defaultValue=1, parameterName=RedAlpha}
  )
  n97 = VertexColor()
  MaterialFunctionCall!(
   In=MakeMaterialAttributes!(
    BaseColor=LinearInterpolate(A=@n1.RGB, Alpha=@n2.R) {constB=0}
    Metallic
    Specular=LinearInterpolate(A=@n3.R, Alpha=@n4.R)
    Roughness=LinearInterpolate(A=@n3.B, Alpha=@n5.R) {constB=0}
    Anisotropy
    EmissiveColor=LinearInterpolate(A=@n47, B=@n96, Alpha=@n97.R)
    Opacity
    OpacityMask
    Normal=TextureSample() {samplerType=SAMPLERTYPE_Normal, texture=/Game/Environments/Props/RiftBeacon/Textures/T_RiftBeaconA_Bits_N}.RGB
    Tangent
    WorldPositionOffset
    SubsurfaceColor
    ClearCoat
    ClearCoatRoughness
    AmbientOcclusion
    Refraction
    CustomizedUVs_0
    CustomizedUVs_1
    CustomizedUVs_2
    CustomizedUVs_3
    CustomizedUVs_4
    CustomizedUVs_5
    CustomizedUVs_6
    CustomizedUVs_7
    PixelDepthOffset
    ShadingModel
    Displacement
   )
   Emissive Mask=TextureObject() {texture=&T_RiftBeaconA_Bits_E}
  ) {materialFunction=/Game/Effects/Fort_Effects/Materials/Geode/_Functions/MF_Geode_Hoses_Lights_RiftBeacon}
 ]
}
