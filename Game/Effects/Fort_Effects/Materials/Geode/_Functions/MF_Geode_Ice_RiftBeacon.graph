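# Material function that layers frost and ice over incoming material attributes, named for the Geode rift beacon.
# Normals packed in UV1/UV2 plus smoke noise blend ice and frozen color, metallic, specular and roughness.
# A local-Z gradient from an offset pivot, broken up by world-aligned smoke, masks that look, frost normals and Fresnel-tinted emissive.
# The Freeze parameter, smoke and vertex red blend the result over the original.
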MaterialFunction /Game/Effects/Fort_Effects/Materials/Geode/_Functions/MF_Geode_Ice_RiftBeacon
graph {nodes [
  ConstantBiasScale(None=Sine(None=Time()) {period=4})
  n1 = FunctionInput() {bUsePreviewValueAsDefault=true, description=spec color goes into world position, inputName=MA, inputType=FunctionInput_MaterialAttributes}
  n2 = GetMaterialAttributes(None=@n1) {attributeGetTypes [
    #1
    #2
    #3
    #4
    #5
    #6
   ]}
  n3 = VectorParameter() {group=Frozen, parameterName=Ice Color, defaultValue=<rgba 0.0084 0.0358 0.13 0>}
  n4 = LinearInterpolate(
   A=VectorParameter() {group=Frozen, parameterName=FrozenBaseColorLow, defaultValue=<rgba 0.5666 0.7257 0.865 0>}.RGB
   B=VectorParameter() {group=Frozen, parameterName=FrozenBaseColorHigh, defaultValue=<rgba 0.705 0.8962 1 0>}.RGB
   Alpha=Desaturation(None=@n2.BaseColor)
  )
  n5 = ComponentMask(None=AppendVector(A=TextureCoordinate() {coordinateIndex=1}, B=TextureCoordinate() {coordinateIndex=2})) {b=true, g=true, r=true}
  n6 = VectorParameter() {group=Frozen, parameterName=Transition Normals Bias, defaultValue=<rgba 0 0 0.75 0>}
  n7 = DotProduct(A=Saturate(None=Add(A=@n5, B=@n6.RGB) {desc=bias the normals}), B=Constant3Vector() {constant=<rgba 0 0 1 0>})
  n8 = ScalarParameter() {group=Frozen, parameterName=Random}
  n9 = Multiply(
   A=TextureCoordinate()
   B=AppendVector(
    A=ScalarParameter() {defaultValue=3, group=Frozen, parameterName=Transition Noise U}
    B=ScalarParameter() {defaultValue=3, group=Frozen, parameterName=Transition Noise V}
   )
  )
  n10 = Multiply(
   A=TextureSample(UVs=Add(A=@n8, B=@n9)) {texture=/Game/Global/Textures/Default/Tileable_Patterns/T_Smoke_OUT}.G
   B=ScalarParameter() {defaultValue=1, group=Frozen, parameterName=Transition Noise Scale}
  )
  n11 = Power(Base=Saturate(None=Add(A=@n7, B=@n10)), Exp=ScalarParameter() {defaultValue=1, group=Frozen, parameterName=Transition Exponent})
  n12 = VectorParameter() {group=Frozen, parameterName=FrozenOffset, defaultValue=<rgba 0 0 1600 0>}
  n13 = Subtract(
   A=Add(
    A=MaterialFunctionCall() {materialFunction=&ObjectPivotPoint}.Object Pivot Location
    B=Transform(None=@n12.RGB) {transformSourceType=TRANSFORMSOURCE_Local}
   )
   B=WorldPosition().XYZ
  )
  n14 = TextureSample(UVs=TextureCoordinate()) {texture=&T_Smoke_OUT}
  n15 = Constant() {r=100}
  n16 = ComponentMask(None=Add(
    A=Transform(None=@n13) {transformSourceType=TRANSFORMSOURCE_World, transformType=TRANSFORM_Local}
    B=Multiply(A=@n14.G, B=@n15) {constB=0.025}
   )) {b=true}
  n17 = Distance(A=@n16, B=ScalarParameter() {defaultValue=150, group=Frozen, parameterName=Ice Gradient Fade Distance})
  n18 = ScalarParameter() {defaultValue=800, group=Frozen, parameterName=Ice Threshold Distance}
  n19 = OneMinus(None=MaterialFunctionCall(Alpha=Divide(A=@n17, B=@n18), Min=Constant(), Max=Constant() {r=1}) {materialFunction=&SmoothStep}.Result)
  n20 = Add(A=Clamp(None=@n16), B=Clamp(None=Multiply(A=@n19) {constB=5}))
  n21 = Desaturation(None=MaterialFunctionCall(
    TextureObject=TextureObject() {texture=&T_Inky_Smoke_Tile}
    TextureSize=ScalarParameter() {defaultValue=2048, group=Frozen, parameterName=WorldAlignedTextureSize}
   ) {materialFunction=&WorldAlignedTexture}.XYZ Texture)
  n22 = ScalarParameter() {defaultValue=1, group=Frozen, parameterName=Freeze_Exp}
  n23 = LinearInterpolate(Alpha=Multiply(A=Clamp(None=@n20), B=Power(Base=@n21, Exp=@n22)))
  n24 = ScalarParameter() {defaultValue=0.8, group=Frozen, parameterName=Ice Metallic}
  n25 = ScalarParameter() {defaultValue=0.3, group=Frozen, parameterName=Frozen Metallic}
  n26 = ScalarParameter() {defaultValue=0.2, group=Frozen, parameterName=Ice Specular}
  n27 = ScalarParameter() {defaultValue=0.8, group=Frozen, parameterName=Frozen Specular}
  n28 = ScalarParameter() {defaultValue=0.6, group=Frozen, parameterName=Ice Roughness}
  n29 = ScalarParameter() {defaultValue=0.15, group=Frozen, parameterName=Frozen Roughness}
  n30 = TextureSampleParameter2D() {group=Frozen, parameterName=Frost Normal, samplerType=SAMPLERTYPE_Normal, texture=/Game/Environments/Sets/Arid/Textures/DesertSand_normal}
  n31 = Normalize(VectorInput=Add(A=@n2.Normal, B=@n30.RGB))
  n32 = LinearInterpolate(
   A=VectorParameter() {group=Frozen, parameterName=Frozen AO Color Low Outer, defaultValue=<rgba 0.0975 0.1247 0.195 0>}.RGB
   B=VectorParameter() {group=Frozen, parameterName=Frozen AO Color Low Inner, defaultValue=<rgba 0.0175 0.0214 0.035 1>}.RGB
   Alpha=MaterialFunctionCall(
    Brightness=ScalarParameter() {defaultValue=1.4, group=Frozen, parameterName=Emissive Fresnel Brightness}
    Exponent=ScalarParameter() {defaultValue=2, group=Frozen, parameterName=Emissive Fresnel Exponent}
    Vertex / Pixel Normal=@n31
   ) {materialFunction=&MF_Fresnel_FX}.Fresnel
  )
  n33 = SetMaterialAttributes!(
   MaterialAttributes=@n2.MaterialAttributes
   BaseColor=LinearInterpolate(A=@n2.BaseColor, B=LinearInterpolate(A=@n3.RGB, B=@n4, Alpha=@n11), Alpha=@n23)
   Metallic=LinearInterpolate(A=@n2.Metallic, B=LinearInterpolate(A=@n24, B=@n25, Alpha=@n11), Alpha=@n23)
   Specular=LinearInterpolate(A=@n2.Specular, B=LinearInterpolate(A=@n26, B=@n27, Alpha=@n11), Alpha=@n23)
   Roughness=LinearInterpolate(A=@n2.Roughness, B=LinearInterpolate(A=@n28, B=@n29, Alpha=@n11), Alpha=@n23)
   EmissiveColor=LinearInterpolate(A=@n2.EmissiveColor, B=Add(A=@n2.EmissiveColor, B=@n32) {desc=Emissive}, Alpha=@n23)
   Normal=LinearInterpolate(A=@n2.Normal, B=@n31, Alpha=@n23)
  ) {attributeSetTypes [
    #1
    #4
    #5
    #6
    #2
    #3
   ]}
  n34 = MaterialFunctionCall(
   TextureObject=TextureObject() {texture=&T_Inky_Smoke_Tile}
   TextureSize=ScalarParameter() {defaultValue=2048, group=Frozen, parameterName=WorldAlignedTextureSize}
  ) {materialFunction=&WorldAlignedTexture}
  n35 = Multiply(
   A=Power(Base=Desaturation(None=@n34.XYZ Texture), Exp=ScalarParameter() {defaultValue=1, group=Frozen, parameterName=Freeze_Exp})
   B=ScalarParameter() {group=Frozen, parameterName=Freeze}
  )
  n36 = VertexColor()
  FunctionOutput(None=BlendMaterialAttributes(A=@n2.MaterialAttributes, B=@n33, Alpha=Multiply(A=@n35, B=@n36.R)))
 ]}
