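# Material function for the Rift Beacon prop that layers a vibration onto incoming material attributes' World Position Offset.
# A DebugTimeSine call, driven by VibrationSpeed, blends between negative and positive VibrationAmount on world Z.
# VibrationOn scales the motion and defaults to off, and red vertex color masks which vertices move.
# The result is added to existing World Position Offset before output.
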MaterialFunction /Game/Environments/Props/RiftBeacon/Materials/MF_RiftBeacon_Vibrations
graph {nodes [
  n1 = GetMaterialAttributes(None=FunctionInput() {inputType=FunctionInput_MaterialAttributes}) {attributeGetTypes [#1]}
  n2 = Constant2Vector()
  n3 = ScalarParameter() {defaultValue=0.5, parameterName=VibrationAmount}
  n4 = AppendVector(A=@n2, B=Multiply(A=@n3) {constB=-1})
  n5 = AppendVector(A=@n2, B=@n3)
  n6 = MaterialFunctionCall(Speed=ScalarParameter() {defaultValue=10, parameterName=VibrationSpeed}) {materialFunction=&DebugTimeSine}
  n7 = LinearInterpolate(
   B=Multiply(A=LinearInterpolate(A=@n4, B=@n5, Alpha=@n6.Result), B=ScalarParameter() {group=Vibrations, parameterName=VibrationOn})
   Alpha=VertexColor().R
  )
  FunctionOutput(None=SetMaterialAttributes!(MaterialAttributes=@n1.MaterialAttributes, WorldPositionOffset=Add(A=@n1.WorldPositionOffset, B=@n7)) {attributeSetTypes [#1]})
 ]}
