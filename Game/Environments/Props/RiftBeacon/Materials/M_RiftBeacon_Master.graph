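# Master material for Rift Beacon props, using material attributes and the Default Lit shading model.
# Four UV-mapped texture sets with tints, masks, normals and scrolling emissive feed a four-layer lerp function.
# Deferred shading adds fuzz, time-of-day emissive, gradient, vertex-paint and up-vector functions plus opacity mask, while Forward and Mobile use a simplified attribute set.
# Both paths then pass through geode ice, game-wide shared and beacon vibration functions.
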Material /Game/Environments/Props/RiftBeacon/Materials/M_RiftBeacon_Master
graph {
 properties {bUseMaterialAttributes=true, ditheredLODTransition=true, pixelDepthOffsetMode=PDOM_Legacy, usedShadingModels=Default Lit}
 nodes [
  n1 = MaterialFunctionCall() {materialFunction=/Game/Athena/WinterArt/Materials/MF_UVsSet}
  n2 = TextureSample(
   UVs=@n1.UVs Layer 1
   Tex=TextureObjectParameter() {group=Material_1, parameterName=Diffuse, texture=/Game/Environments/Sets/IceCastle/Textures/T_IceCastle_Wall1_D}
  ) {texture=&T_IceCastle_Wall1_D}
  n3 = MaterialFunctionCall!(In=VectorParameter() {group=Material_1, parameterName=TintColor, sortPriority=7, defaultValue=<rgba 1 1 1 0>}.RGB) {materialFunction=/Game/Athena/WinterArt/Materials/MF_VPTintMask}
  n4 = ScalarParameter() {defaultValue=1, group=Material_1, parameterName=L1 Alpha, sortPriority=1}
  n5 = MaterialFunctionCall() {materialFunction=&MF_UVsSet}
  n6 = TextureSampleParameter2D(UVs=@n5.UVs Layer 2) {
   desc=Use for interior walls and ceilings
   group=Material_2
   parameterName=Diffuse_Texture_2
   samplerSource=SSM_Wrap_WorldGroupSettings
   sortPriority=1
   texture=/Game/Global/Textures/Default/Blanks/T_MidgreysRGB
  }
  n7 = MaterialFunctionCall(In=VectorParameter() {group=Material_2, parameterName=M2_TintColor, sortPriority=5, defaultValue=<rgba 1 1 1 0>}.RGB) {materialFunction=&MF_VPTintMask}
  n8 = ScalarParameter() {defaultValue=1, group=Material_2, parameterName=L2 Alpha, sortPriority=1}
  n9 = MaterialFunctionCall() {materialFunction=&MF_UVsSet}
  n10 = TextureSampleParameter2D(UVs=@n9.UVs Layer 3) {
   desc=Use for trims and accents
   group=Material_3
   parameterName=Diffuse_Texture_3
   samplerSource=SSM_Wrap_WorldGroupSettings
   sortPriority=1
   texture=&T_MidgreysRGB
  }
  n11 = MaterialFunctionCall(In=VectorParameter() {group=Material_3, parameterName=M3_TintColor, sortPriority=5, defaultValue=<rgba 1 1 1 0>}.RGB) {materialFunction=&MF_VPTintMask}
  n12 = ScalarParameter() {defaultValue=1, group=Material_3, parameterName=L3 Alpha, sortPriority=1}
  n13 = MaterialFunctionCall() {materialFunction=&MF_UVsSet}
  n14 = TextureSampleParameter2D(UVs=@n13.UVs Layer 4) {
   desc=Use for doors
   group=Material_4
   parameterName=Diffuse_Texture_4
   samplerSource=SSM_Wrap_WorldGroupSettings
   sortPriority=1
   texture=&T_MidgreysRGB
  }
  n15 = MaterialFunctionCall(In=VectorParameter() {group=Material_4, parameterName=M4_TintColor, sortPriority=5, defaultValue=<rgba 1 1 1 0>}.RGB) {materialFunction=&MF_VPTintMask}
  n16 = ScalarParameter() {defaultValue=1, group=Material_4, parameterName=L4 Alpha, sortPriority=1}
  n17 = TextureObjectParameter() {group=Material_1, parameterName=Normals, samplerType=SAMPLERTYPE_Normal, sortPriority=2, texture=/Engine/EngineMaterials/FlatNormal}
  n18 = TextureObjectParameter() {
   group=Material_1
   parameterName=SpecularMasks
   samplerType=SAMPLERTYPE_Masks
   sortPriority=1
   texture=&T_Fortnite_Default_S
  }
  n19 = TextureSample(UVs=@n1.UVs Layer 1, Tex=@n18) {samplerType=SAMPLERTYPE_Masks, texture=&T_IceCastle_Wall1_D}
  n20 = TextureSampleParameter2D(UVs=@n13.UVs Layer 4) {
   desc=Use for doors
   group=Material_4
   parameterName=SpecularMasks_4
   samplerSource=SSM_Wrap_WorldGroupSettings
   samplerType=SAMPLERTYPE_Masks
   sortPriority=2
   texture=&T_Fortnite_Default_S
  }
  n21 = TextureSampleParameter2D(UVs=@n13.UVs Layer 4) {
   desc=Use for doors
   group=Material_4
   parameterName=Normals_Texture_4
   samplerSource=SSM_Wrap_WorldGroupSettings
   samplerType=SAMPLERTYPE_Normal
   sortPriority=3
   texture=&FlatNormal
  }
  n22 = TextureSampleParameter2D(UVs=@n5.UVs Layer 2) {
   desc=Use for interior walls and ceilings
   group=Material_2
   parameterName=SpecularMasks_2
   samplerSource=SSM_Wrap_WorldGroupSettings
   samplerType=SAMPLERTYPE_Masks
   sortPriority=2
   texture=&T_Fortnite_Default_S
  }
  n23 = TextureSampleParameter2D(UVs=@n5.UVs Layer 2) {
   desc=Use for interior walls and ceilings
   group=Material_2
   parameterName=Normals_Texture_2
   samplerSource=SSM_Wrap_WorldGroupSettings
   samplerType=SAMPLERTYPE_Normal
   sortPriority=3
   texture=&FlatNormal
  }
  n24 = TextureSampleParameter2D(UVs=@n9.UVs Layer 3) {
   desc=Use for trims and accents
   group=Material_3
   parameterName=SpecularMasks_3
   samplerSource=SSM_Wrap_WorldGroupSettings
   samplerType=SAMPLERTYPE_Masks
   sortPriority=2
   texture=&T_Fortnite_Default_S
  }
  n25 = TextureSampleParameter2D(UVs=@n9.UVs Layer 3) {
   desc=Use for trims and accents
   group=Material_3
   parameterName=Normals_Texture_3
   samplerSource=SSM_Wrap_WorldGroupSettings
   samplerType=SAMPLERTYPE_Normal
   sortPriority=3
   texture=&FlatNormal
  }
  n26 = MaterialFunctionCall!(
   Use3Materials
   Use4Materials
   Use2Materials
   D=AppendVector(A=Multiply(A=@n2.RGB, B=@n3.Result), B=Multiply(A=@n2.A, B=@n4))
   S=AppendVector(A=@n19.RGB, B=@n19.RGB)
   N=TextureSample(UVs=@n1.UVs Layer 1, Tex=@n17) {samplerType=SAMPLERTYPE_Normal}.RGB
   D2=AppendVector(A=Multiply(A=@n6.RGB, B=@n7.Result), B=Multiply(A=@n8, B=@n6.A))
   S2=AppendVector(A=@n22.RGB, B=@n22.RGB)
   N2=@n23.RGB
   D3=AppendVector(A=Multiply(A=@n10.RGB, B=@n11.Result), B=Multiply(A=@n12, B=@n10.A))
   S3=AppendVector(A=@n24.RGB, B=@n24.RGB)
   N3=@n25.RGB
   D4=AppendVector(A=Multiply(A=@n14.RGB, B=@n15.Result), B=Multiply(A=@n16, B=@n14.A))
   S4=AppendVector(A=@n20.RGB, B=@n20.RGB)
   N4=@n21.RGB
   E1=MaterialFunctionCall!(
    X=ScalarParameter() {group=Adv Emissive, parameterName=L1 X Speed, sortPriority=12}
    Y=ScalarParameter() {group=Adv Emissive, parameterName=L1 Y Speed, sortPriority=13}
    Dot=VectorParameter() {group=Adv Emissive, parameterName=L1 Channel, sortPriority=9}.RGB
    TextureOffset=ScalarParameter() {defaultValue=1, group=Adv Emissive, parameterName=L1 Offset, sliderMax=4, sliderMin=1, sortPriority=11}
    EmissiveTexture=VectorParameter() {group=Material_1, parameterName=Emissive, sortPriority=7}.RGB
   ) {materialFunction=/Game/Athena/WinterArt/Materials/MF_EmissiveMask}.EmissiveMasked
   E2=MaterialFunctionCall(
    X=ScalarParameter() {group=Adv Emissive, parameterName=L2 X Speed, sortPriority=16}
    Y=ScalarParameter() {group=Adv Emissive, parameterName=L2 Y Speed, sortPriority=17}
    Dot=VectorParameter() {group=Adv Emissive, parameterName=L2 Channel, sortPriority=14}.RGB
    TextureOffset=ScalarParameter() {defaultValue=1, group=Adv Emissive, parameterName=L2 Offset, sliderMax=4, sliderMin=1, sortPriority=15}
    EmissiveTexture=VectorParameter() {group=Material_2, parameterName=Emissive2, sortPriority=7}.RGB
   ) {materialFunction=&MF_EmissiveMask}.EmissiveMasked
   E3=MaterialFunctionCall(
    X=ScalarParameter() {group=Adv Emissive, parameterName=L3 X Speed, sortPriority=21}
    Y=ScalarParameter() {group=Adv Emissive, parameterName=L3 Y Speed, sortPriority=22}
    Dot=VectorParameter() {group=Adv Emissive, parameterName=L3 Channel, sortPriority=18}.RGB
    TextureOffset=ScalarParameter() {defaultValue=1, group=Adv Emissive, parameterName=L3 Offset, sliderMax=4, sliderMin=1, sortPriority=19}
    EmissiveTexture=VectorParameter() {group=Material_3, parameterName=Emissive3, sortPriority=7}.RGB
   ) {materialFunction=&MF_EmissiveMask}.EmissiveMasked
   E4=MaterialFunctionCall(
    X=ScalarParameter() {group=Adv Emissive, parameterName=L4 X Speed, sortPriority=25}
    Y=ScalarParameter() {group=Adv Emissive, parameterName=L4 Y Speed, sortPriority=26}
    Dot=VectorParameter() {group=Adv Emissive, parameterName=L4 Channel, sortPriority=23}.RGB
    TextureOffset=ScalarParameter() {defaultValue=1, group=Adv Emissive, parameterName=L4 Offset, sliderMax=4, sliderMin=1, sortPriority=24}
    EmissiveTexture=VectorParameter() {group=Material_4, parameterName=Emissive4, sortPriority=7}.RGB
   ) {materialFunction=&MF_EmissiveMask}.EmissiveMasked
   CCMaskByMetal_1
   CC1=ScalarParameter() {group=Material_1, parameterName=L1 Clearcoat}
   CCR_1
   CC2=ScalarParameter() {group=Material_2, parameterName=L2 Clearcoat}
   CCR_2
   CCMaskByMetal_2
   CC3=ScalarParameter() {group=Material_3, parameterName=L3 Clearcoat}
   CCR_3
   CCMaskByMetal_3
   CC4=ScalarParameter() {group=Material_4, parameterName=L4 Clearcoat}
   CCMaskByMetal_4
   CCR_4
   OpacityMaskCancel_1
   OpacityMaskCancel_2
   OpacityMaskCancel_3
   OpacityMaskCancel_4
   SSS_1
   SSS_2
   SSS_3
   SSS_4
   SSS_Depth_1
   SSS_Depth_2
   SSS_Depth_3
   SSS_Depth_4
  ) {materialFunction=/Game/Athena/WinterArt/Materials/MF_FN_TextureSetLerp_4Layers}
  n27 = SetMaterialAttributes!(
   MaterialAttributes
   Base Color=@n26.Diffuse
   Specular=@n26.Diffuse
   Metallic=@n26.Diffuse
   Roughness=@n26.Diffuse
   Emissive Color=@n26.Diffuse
   Normal=@n26.Diffuse
   Custom Data 0=Multiply(A=@n26.Clearcoat, B=@n26.Height)
   Custom Data 1=Constant() {r=0.0125}
   World Position Offset=Constant()
  ) {attributeSetTypes [
    #1
    #2
    #3
    #4
    #5
    #6
    #7
    #8
    #9
   ]}
  n28 = MaterialFunctionCall!(In=@n27) {materialFunction=/Game/Athena/WinterArt/Materials/MF_Fuzzy}
  n29 = MaterialFunctionCall!(
   In=MaterialFunctionCall!(
    transitionFactor
    MaterialIn=MaterialFunctionCall!(In=@n28.Result) {materialFunction=/Game/Athena/WinterArt/Materials/MF_TOD_EmissiveColors}.Result
   ) {materialFunction=/Game/Athena/WinterArt/Materials/MF_WSGradient}.Color*Gradient
   UseOneSampler
  ) {materialFunction=/Game/Athena/WinterArt/Materials/MF_VPTexture}
  n30 = SetMaterialAttributes!(
   MaterialAttributes=MaterialFunctionCall!(In=@n29.Result, UpVectorTexture=@n29.Result) {materialFunction=/Game/Athena/WinterArt/Materials/MF_UpVectorTexture}.Result
   Opacity Mask=@n26.Height
  ) {attributeSetTypes [#10]}
  n31 = SetMaterialAttributes!(MaterialAttributes, Base Color=@n26.Diffuse, Specular, Metallic, Roughness, Emissive Color, Normal) {attributeSetTypes [
    #1
    #2
    #3
    #4
    #5
    #6
   ]}
  n32 = ShadingPathSwitch(Default=@n30, Deferred=@n30, Forward=@n31, Mobile=@n31)
  MaterialFunctionCall!(In=MaterialFunctionCall!(In=MaterialFunctionCall!(MA=@n32) {materialFunction=/Game/Effects/Fort_Effects/Materials/Geode/_Functions/MF_Geode_Ice_RiftBeacon}.Result) {materialFunction=/Game/Athena/WinterArt/Materials/MF_GameWideShared}.Result) {materialFunction=&/MF_RiftBeacon_Vibrations}
  TextureObjectParameter() {group=Material_2, parameterName=Diffuse_Texture_2, sortPriority=1, texture=&T_MidgreysRGB}
  TextureObjectParameter() {group=Material_3, parameterName=Diffuse_Texture_3, sortPriority=1, texture=&T_MidgreysRGB}
  TextureObjectParameter() {group=Material_4, parameterName=Diffuse_Texture_4, sortPriority=1, texture=&T_MidgreysRGB}
 ]
}
