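# Material function for the RiftBeacon prop that computes a world position offset for a bobbing, spinning nexus.
# A time sine blends between negative and positive offset vectors, built from the bob amount on Z plus scaled red vertex color on all axes.
# The Splines parameter fades that bob toward zero, and the result is added to a time-driven rotation about the object pivot on a configurable axis, Z by default.
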MaterialFunction /Game/Environments/Props/RiftBeacon/FX/RiftBeacon_NexusWPO
graph {nodes [
  n1 = Multiply(A=VertexColor().R, B=ScalarParameter() {defaultValue=1, parameterName=VertColorScale})
  n2 = Add(A=Constant2Vector(), B=@n1)
  n3 = Add(A=ScalarParameter() {defaultValue=5, parameterName=BOB Amount}, B=@n1)
  n4 = LinearInterpolate(
   A=AppendVector(A=Multiply(A=@n2) {constB=-1}, B=Multiply(A=@n3) {constB=-1})
   B=AppendVector(A=@n2, B=@n3)
   Alpha=MaterialFunctionCall(Speed=ScalarParameter() {defaultValue=0.5, parameterName=Bob Speed}) {materialFunction=&DebugTimeSine}.Result
  )
  n5 = ScalarParameter() {parameterName=Splines}
  n6 = VectorParameter() {parameterName=RotationAxis, defaultValue=<rgba 0 0 1 0>}
  n7 = Multiply(A=Time(), B=ScalarParameter() {defaultValue=0.05, parameterName=RotationSpeed})
  n8 = MaterialFunctionCall() {materialFunction=&ObjectPivotPoint}
  n9 = WorldPosition()
  FunctionOutput(None=Add(
    A=LinearInterpolate(A=@n4, Alpha=@n5) {constB=0}
    B=RotateAboutAxis(NormalizedRotationAxis=@n6.RGB, RotationAngle=@n7, PivotPoint=@n8.Object Pivot Location, Position=@n9.XYZ)
   ))
 ]}
