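# Material function that adds an audio-reactive "TacoTime" emissive pattern to the Rift Beacon's green goop bubble, blended over Base Emissive by the TACOTIME collection parameter.
# Collection time and low-audio values, with a fake-audio debug override, drive stepped hue shifts and brightness pulses.
# A panning Marshmello hat mask offsets the hue, while time cycles its scale and speed.
# An inverted Fresnel masks the glow; two FortniteMaterialParameters nodes are unconnected.
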MaterialFunction /Game/Environments/Props/RiftBeacon/FX/RiftBubble_GreenGoop
graph {nodes [
  CollectionParameter() {collection=&FortniteMaterialParameters, parameterName=Emote_AudioLow}
  CollectionParameter() {collection=&FortniteMaterialParameters, parameterName=Emote_EnvTime}
  n1 = ConstantBiasScale(None=Sine(None=Time()) {period=2}) {bias=0.5, scale=1}
  n2 = Constant() {r=2}
  n3 = LinearInterpolate(
   A=ScalarParameter() {group=TacoTime_Debug, parameterName=FakeAudioData, sortPriority=1}
   B=Clamp(None=Power(Base=@n1, Exp=@n2)) {minDefault=0.25}
   Alpha=ScalarParameter() {group=TacoTime_Debug, parameterName=FakeAudioData_Fade}
  )
  n4 = LinearInterpolate(
   A=CollectionParameter() {collection=&/MPC_RiftBubble_GreenGoop, parameterName=TacoTime_EnvTime}
   Alpha=@n3
  )
  n5 = Constant() {r=8}
  n6 = Ceil(None=Multiply(A=@n4, B=@n5) {constB=8})
  n7 = FunctionInput() {bUsePreviewValueAsDefault=true, inputName=Offset, inputType=FunctionInput_Scalar, previewValue=<0.5 0 0 1>}
  n8 = Power(
   Base=LinearInterpolate(
    A=CollectionParameter() {collection=&MPC_RiftBubble_GreenGoop, parameterName=TacoTime_AudioLow}
    Alpha=@n3
   )
   Exp=ScalarParameter() {defaultValue=2, group=TacoTime, parameterName=Pattern 1 Pulse Power, sortPriority=11}
  )
  n9 = Divide(A=Ceil(None=Multiply(A=@n8) {constB=8})) {constB=8}
  n10 = Constant() {r=2}
  n11 = Add(A=Divide(A=Divide(A=@n6, B=@n5) {constB=8}) {constB=2.1111}, B=Add(A=Divide(A=@n7) {constB=1.5}, B=Divide(A=@n9, B=@n10) {constB=20}))
  n12 = Constant() {r=128}
  n13 = Round(None=Frac(None=Divide(A=@n4, B=@n12) {constB=2.1111}))
  n14 = Divide(
   A=TextureCoordinate()
   B=LinearInterpolate(
    A=ScalarParameter() {defaultValue=100, group=TacoTime, parameterName=Pattern 1 Panner Spread 1, sortPriority=6}
    B=ScalarParameter() {defaultValue=40, group=TacoTime, parameterName=Pattern 1 Panner Spread 2, sortPriority=7}
    Alpha=@n13
   )
  )
  n15 = VectorParameter() {group=TacoTime, parameterName=Panning Speed, sortPriority=8, defaultValue=<rgba 0 400 0 1>}
  n16 = VectorParameter() {group=TacoTime, parameterName=Panning Speed2, sortPriority=9, defaultValue=<rgba 0 0 -5 1>}
  n17 = Frac(None=Divide(A=@n4, B=Constant() {r=32}) {constB=2.1111})
  n18 = Frac(None=Multiply(A=Time(), B=LinearInterpolate(A=@n15.RGB, B=@n16.RGB, Alpha=@n17)))
  n19 = TextureSampleParameter2D(UVs=Add(A=ComponentMask(None=@n14) {g=true, r=true}, B=ComponentMask(None=@n18) {b=true, g=true})) {
   group=TacoTime
   parameterName=Wavy
   samplerType=SAMPLERTYPE_Masks
   sortPriority=5
   texture=/Game/Accessories/Hats/M_MED_Marshmellow_DJ_Hat/Textures/Mello_M
  }
  n20 = LinearInterpolate(A=@n19.R, B=@n19.R, Alpha=@n13)
  n21 = Constant3Vector() {constant=<rgba 0 1 0.3489 1>}
  n22 = Multiply(
   A=MaterialFunctionCall(Hue Shift Percentage=Add(A=@n11, B=@n20), Texture=FunctionInput(Preview=@n21) {bUsePreviewValueAsDefault=true, inputName=StartingHue}) {materialFunction=&HueShift}.Result
   B=Multiply(A=@n8, B=ScalarParameter() {defaultValue=2500, group=TacoTime, parameterName=Pattern 1 Brightness, sortPriority=10}) {constB=2}
  )
  n23 = Fresnel(ExponentIn=ScalarParameter() {defaultValue=0.0025, group=TacoTime, parameterName=Emissive Fresnel, sortPriority=1}) {baseReflectFraction=0}
  FunctionOutput(None=LinearInterpolate(
    A=FunctionInput() {bUsePreviewValueAsDefault=true, inputName=Base Emissive}
    B=LinearInterpolate(A=@n22, Alpha=@n23) {constB=0}
    Alpha=CollectionParameter() {collection=&MPC_RiftBubble_GreenGoop, parameterName=TACOTIME}
   )) {outputName=TacoTimeResults}
 ]}
